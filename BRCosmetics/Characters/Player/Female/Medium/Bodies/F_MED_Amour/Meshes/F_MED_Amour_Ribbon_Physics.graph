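# Physics asset for the back bow and ribbons on the F_MED_Amour female body cosmetic, using one capsule per bone.
# The bow center is tightly limited to spine_02 and parents two twist-locked side loops and two three-link ribbon chains with limited swing and twist.
# Pelvis, lower spine and two back-skirt capsules are joined by fully free constraints, which suggests they serve as collision shapes for the ribbons.
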PhysicsAsset /BRCosmetics/Characters/Player/Female/Medium/Bodies/F_MED_Amour/Meshes/F_MED_Amour_Ribbon_Physics
bodies {
 dyn_bow_main [{
   shapeType=Capsule
   length=3.0042
   radius=2.9386
   shapeName=dyn_bow_main_capsule
   center=<-0.1864 0.228 0.0225>
   rotation=<rot -29.6225 -101.3989 -2.788>
  }]
 dyn_bow_ribbon_l_01 [{
   shapeType=Capsule
   length=5
   radius=2
   shapeName=dyn_bow_ribbon_l_01_capsule
   center=<0 -0.0713 0.8151>
   rotation=<rot -5.0002 -89.9996 -89.9991>
  }]
 dyn_bow_ribbon_l_02 [{
   shapeType=Capsule
   length=5
   radius=2
   shapeName=dyn_bow_ribbon_l_02_capsule
   center=<1.0336 -0.0713 0.8151>
   rotation=<rot -5.0002 -89.9996 -89.9991>
  }]
 dyn_bow_ribbon_l_03 [{
   shapeType=Capsule
   length=5.25
   radius=2.75
   shapeName=dyn_bow_ribbon_l_03_capsule
   center=<2.5776 -0.127 1.4521>
   rotation=<rot -5.0002 -89.9996 -89.9991>
  }]
 dyn_bow_ribbon_r_01 [{
   shapeType=Capsule
   length=5
   radius=2
   shapeName=dyn_bow_ribbon_r_01_capsule
   center=<0 0.0878 0.3278>
   rotation=<rot 15.0002 -90.0003 -89.999>
  }]
 dyn_bow_ribbon_r_02 [{
   shapeType=Capsule
   length=5
   radius=2
   shapeName=dyn_bow_ribbon_r_02_capsule
   center=<0.5651 0.0878 0.3278>
   rotation=<rot 15.0002 -90.0003 -89.999>
  }]
 dyn_bow_ribbon_r_03 [{
   shapeType=Capsule
   length=5.5
   radius=2.75
   shapeName=dyn_bow_ribbon_r_03_capsule
   center=<2.5496 0.2533 0.9455>
   rotation=<rot 15.0002 -90.0003 -89.999>
  }]
 dyn_bow_side_l [{
   shapeType=Capsule
   length=3.3294
   radius=3.406
   shapeName=dyn_bow_side_l_capsule
   center=<-0.8962 -0.3154 1.3311>
   rotation=<rot 15.7963 -90 -30.3356>
  }]
 dyn_bow_side_r [{
   shapeType=Capsule
   length=3.0794
   radius=3.1864
   shapeName=dyn_bow_side_r_capsule
   center=<-1.1435 -1.6113 0.6694>
   rotation=<rot 0 150.1709 -89.0927>
  }]
 dyn_skirt_bck_mid_01 [{
   shapeType=Capsule
   length=11.1411
   radius=6.5401
   shapeName=dyn_skirt_bck_mid_01_capsule
   center=<-0.9962 0 -7.6032>
   rotation=<rot 9.9998 0 90>
  }]
 dyn_skirt_bck_mid_02 [{
   shapeType=Capsule
   length=12.4531
   radius=12.4351
   shapeName=dyn_skirt_bck_mid_02_capsule
   center=<-0.4418 1.1143 -11.2925>
   rotation=<rot -10.0002 0 -90>
  }]
 pelvis [{shapeType=Capsule, length=18.1641, radius=8.5, shapeName=pelvis_capsule, center=<-0.1409 -2.8134 0>, rotation=<rot 0 -100.0002 0>}]
 spine_01 [{
   shapeType=Capsule
   length=15.257
   radius=6.9122
   shapeName=spine_01_capsule
   center=<4.7496 1.8246 -0.5868>
   rotation=<rot 5.9495 -100.0468 1.2732>
  }]
 spine_02 [{
   shapeType=Capsule
   length=12.3587
   radius=7.5666
   shapeName=spine_02_capsule
   center=<2.8122 2.5979 0.0126>
   rotation=<rot -1.494 -100.0034 0.6343>
  }]
}
constraints [
 {
  bone1Name=spine_01
  bone2Name=pelvis
  swing1LimitDegrees=45
  swing1Motion=Free
  swing2LimitDegrees=45
  swing2Motion=Free
  twistLimitDegrees=45
  twistMotion=Free
 }
 {
  bone1Name=dyn_skirt_bck_mid_01
  bone2Name=pelvis
  swing1LimitDegrees=45
  swing1Motion=Free
  swing2LimitDegrees=45
  swing2Motion=Free
  twistLimitDegrees=45
  twistMotion=Free
 }
 {
  bone1Name=dyn_skirt_bck_mid_02
  bone2Name=dyn_skirt_bck_mid_01
  swing1LimitDegrees=45
  swing1Motion=Free
  swing2LimitDegrees=45
  swing2Motion=Free
  twistLimitDegrees=45
  twistMotion=Free
 }
 {
  bone1Name=spine_02
  bone2Name=spine_01
  swing1LimitDegrees=45
  swing1Motion=Free
  swing2LimitDegrees=45
  swing2Motion=Free
  twistLimitDegrees=45
  twistMotion=Free
 }
 {bone1Name=dyn_bow_main, bone2Name=spine_02, swing1LimitDegrees=1, swing2LimitDegrees=1, twistLimitDegrees=0}
 {bone1Name=dyn_bow_ribbon_l_01, bone2Name=dyn_bow_main, swing1LimitDegrees=15, swing2LimitDegrees=45, twistLimitDegrees=5}
 {bone1Name=dyn_bow_ribbon_l_02, bone2Name=dyn_bow_ribbon_l_01, swing1LimitDegrees=15, swing2LimitDegrees=45, twistLimitDegrees=5}
 {bone1Name=dyn_bow_ribbon_l_03, bone2Name=dyn_bow_ribbon_l_02, swing1LimitDegrees=15, swing2LimitDegrees=45, twistLimitDegrees=5}
 {bone1Name=dyn_bow_ribbon_r_01, bone2Name=dyn_bow_main, swing1LimitDegrees=15, swing2LimitDegrees=45, twistLimitDegrees=5}
 {bone1Name=dyn_bow_ribbon_r_02, bone2Name=dyn_bow_ribbon_r_01, swing1LimitDegrees=15, swing2LimitDegrees=45, twistLimitDegrees=5}
 {bone1Name=dyn_bow_ribbon_r_03, bone2Name=dyn_bow_ribbon_r_02, swing1LimitDegrees=15, swing2LimitDegrees=45, twistLimitDegrees=5}
 {bone1Name=dyn_bow_side_l, bone2Name=dyn_bow_main, swing1LimitDegrees=10, swing2LimitDegrees=5, twistLimitDegrees=45, twistMotion=Locked}
 {bone1Name=dyn_bow_side_r, bone2Name=dyn_bow_main, swing1LimitDegrees=10, swing2LimitDegrees=5, twistLimitDegrees=45, twistMotion=Locked}
]
