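# Physics asset for the second skirt of the F_MED_Amour female medium cosmetic body, defining its bodies and joints.
# Seven two-bone skirt chains (front, side and back, plus back middle) use identical-sized capsules, each chain hanging from the pelvis.
# Skirt joints allow swing on one axis only, with zero twist, so each chain bends like a limited hinge.
# Pelvis and thigh capsules represent the body, with thighs swinging and twisting freely from the pelvis.
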PhysicsAsset /BRCosmetics/Characters/Player/Female/Medium/Bodies/F_MED_Amour/Meshes/F_MED_Amour_Skirt2_Physics
bodies {
 dyn_skirt_bck_l_01 [{
   shapeType=Capsule
   length=7.5
   radius=5.75
   shapeName=dyn_skirt_bck_l_01_capsule
   center=<8.0055 0 4.4517>
   rotation=<rot -74.9997 179.9999 -89.9996>
  }]
 dyn_skirt_bck_l_02 [{
   shapeType=Capsule
   length=7.5
   radius=5.75
   shapeName=dyn_skirt_bck_l_02_capsule
   center=<7.8522 0 2.6994>
   rotation=<rot -74.9996 179.9998 -89.9996>
  }]
 dyn_skirt_bck_mid_01 [{
   shapeType=Capsule
   length=7.5
   radius=5.75
   shapeName=dyn_skirt_bck_mid_01_capsule
   center=<8.0055 0 4.4517>
   rotation=<rot -74.9997 180 -180>
  }]
 dyn_skirt_bck_mid_02 [{
   shapeType=Capsule
   length=7.5
   radius=5.75
   shapeName=dyn_skirt_bck_mid_02_capsule
   center=<7.8522 0 2.6993>
   rotation=<rot -74.9997 180 -180>
  }]
 dyn_skirt_bck_r_01 [{
   shapeType=Capsule
   length=7.5
   radius=5.75
   shapeName=dyn_skirt_bck_r_01_capsule
   center=<8.0055 0 4.4517>
   rotation=<rot -74.9996 -179.9998 -90.0006>
  }]
 dyn_skirt_bck_r_02 [{
   shapeType=Capsule
   length=7.5
   radius=5.75
   shapeName=dyn_skirt_bck_r_02_capsule
   center=<7.8522 0 2.6994>
   rotation=<rot -74.9995 -179.9997 -90.0005>
  }]
 dyn_skirt_fr_l_01 [{
   shapeType=Capsule
   length=7.5
   radius=5.75
   shapeName=dyn_skirt_fr_l_01_capsule
   center=<3.3104 0 -7.1462>
   rotation=<rot -80.0001 -0.0002 -0.0001>
  }]
 dyn_skirt_fr_l_02 [{
   shapeType=Capsule
   length=7.5
   radius=5.75
   shapeName=dyn_skirt_fr_l_02_capsule
   center=<3.4907 0 -5.0844>
   rotation=<rot -80.0001 -0.0002 -0.0001>
  }]
 dyn_skirt_fr_r_01 [{
   shapeType=Capsule
   length=7.5
   radius=5.75
   shapeName=dyn_skirt_fr_r_01_capsule
   center=<8.0055 0 4.4517>
   rotation=<rot -75.0002 -180 179.9998>
  }]
 dyn_skirt_fr_r_02 [{
   shapeType=Capsule
   length=7.5
   radius=5.75
   shapeName=dyn_skirt_fr_r_02_capsule
   center=<7.8522 0 2.6994>
   rotation=<rot -75.0002 -180 179.9998>
  }]
 dyn_skirt_side_l_01 [{
   shapeType=Capsule
   length=7.5
   radius=5.75
   shapeName=dyn_skirt_side_l_01_capsule
   center=<8.0055 0 4.4517>
   rotation=<rot -75.0001 179.9997 -89.9993>
  }]
 dyn_skirt_side_l_02 [{
   shapeType=Capsule
   length=7.5
   radius=5.75
   shapeName=dyn_skirt_side_l_02_capsule
   center=<7.8522 0 2.6994>
   rotation=<rot -75 179.9997 -89.9993>
  }]
 dyn_skirt_side_r_01 [{
   shapeType=Capsule
   length=7.5
   radius=5.75
   shapeName=dyn_skirt_side_r_01_capsule
   center=<8.0055 0 4.4517>
   rotation=<rot -75 -179.9997 -90.0007>
  }]
 dyn_skirt_side_r_02 [{
   shapeType=Capsule
   length=7.5
   radius=5.75
   shapeName=dyn_skirt_side_r_02_capsule
   center=<7.8522 0 2.6994>
   rotation=<rot -75 -179.9997 -90.0007>
  }]
 pelvis [{
   shapeType=Capsule
   length=5.5636
   radius=3.4839
   shapeName=pelvis_capsule
   center=<-0.282 -2.2173 -0.2844>
   rotation=<rot -0.1685 -90 -1.5197>
  }]
 thigh_l [{shapeType=Capsule, length=37.75, radius=8.25, shapeName=thigh_l_capsule, center=<-19.3349 0.0001 0>, rotation=<rot 90 160.5289 160.5293>}]
 thigh_r [{shapeType=Capsule, length=37.75, radius=8.25, shapeName=thigh_r_capsule, center=<19.6473 -0.0001 0>, rotation=<rot -90 -10.0249 10.0245>}]
}
constraints [
 {bone1Name=dyn_skirt_bck_l_02, bone2Name=dyn_skirt_bck_l_01, swing1LimitDegrees=0, swing2LimitDegrees=30, twistLimitDegrees=0}
 {bone1Name=dyn_skirt_bck_r_02, bone2Name=dyn_skirt_bck_r_01, swing1LimitDegrees=0, swing2LimitDegrees=30, twistLimitDegrees=0}
 {bone1Name=dyn_skirt_fr_l_02, bone2Name=dyn_skirt_fr_l_01, swing1LimitDegrees=0, swing2LimitDegrees=30, twistLimitDegrees=0}
 {bone1Name=dyn_skirt_fr_r_02, bone2Name=dyn_skirt_fr_r_01, swing1LimitDegrees=0, swing2LimitDegrees=30, twistLimitDegrees=0}
 {bone1Name=dyn_skirt_side_l_02, bone2Name=dyn_skirt_side_l_01, swing1LimitDegrees=0, swing2LimitDegrees=30, twistLimitDegrees=0}
 {bone1Name=dyn_skirt_side_r_02, bone2Name=dyn_skirt_side_r_01, swing1LimitDegrees=0, swing2LimitDegrees=30, twistLimitDegrees=0}
 {
  bone1Name=thigh_l
  bone2Name=pelvis
  swing1LimitDegrees=45
  swing1Motion=Free
  swing2LimitDegrees=45
  swing2Motion=Free
  twistLimitDegrees=45
  twistMotion=Free
 }
 {
  bone1Name=thigh_r
  bone2Name=pelvis
  swing1LimitDegrees=45
  swing1Motion=Free
  swing2LimitDegrees=45
  swing2Motion=Free
  twistLimitDegrees=45
  twistMotion=Free
 }
 {bone1Name=dyn_skirt_fr_r_01, bone2Name=pelvis, swing1LimitDegrees=0, swing2LimitDegrees=30, twistLimitDegrees=0}
 {bone1Name=dyn_skirt_fr_l_01, bone2Name=pelvis, swing1LimitDegrees=0, swing2LimitDegrees=30, twistLimitDegrees=0}
 {bone1Name=dyn_skirt_side_r_01, bone2Name=pelvis, swing1LimitDegrees=0, swing2LimitDegrees=30, twistLimitDegrees=0}
 {bone1Name=dyn_skirt_side_l_01, bone2Name=pelvis, swing1LimitDegrees=0, swing2LimitDegrees=30, twistLimitDegrees=0}
 {bone1Name=dyn_skirt_bck_r_01, bone2Name=pelvis, swing1LimitDegrees=0, swing2LimitDegrees=30, twistLimitDegrees=0}
 {bone1Name=dyn_skirt_bck_l_01, bone2Name=pelvis, swing1LimitDegrees=0, swing2LimitDegrees=30, twistLimitDegrees=0}
 {bone1Name=dyn_skirt_bck_mid_01, bone2Name=pelvis, swing1LimitDegrees=0, swing2LimitDegrees=30, twistLimitDegrees=0}
 {bone1Name=dyn_skirt_bck_mid_02, bone2Name=dyn_skirt_bck_mid_01, swing1LimitDegrees=0, swing2LimitDegrees=30, twistLimitDegrees=0}
]
