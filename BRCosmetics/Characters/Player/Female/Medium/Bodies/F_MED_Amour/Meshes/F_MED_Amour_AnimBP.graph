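# Animation Blueprint for the F_MED_Amour female cosmetic body, adding skirt, bow, ribbon and shoulder-ring secondary motion to the pose copied from its attached parent.
# A pre-sim Control Rig, thigh constraints and bone offsets feed a skirt RigidBody, two differently tuned ribbon RigidBodies, further offsets and a post-sim arm-ring Control Rig.
# The update event computes skirt gravity, sine floaters and alpha variables for frontend, backbling, emote and sliding states.
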AnimBlueprint /BRCosmetics/Characters/Player/Female/Medium/Bodies/F_MED_Amour/Meshes/F_MED_Amour_AnimBP
target_skeleton=&/F_MED_Amour_Skeleton
graphs [
 AnimGraph {nodes [
   n1 = Control Rig(Source=Copy Pose From Mesh() {bUseAttachedParent=true}) {
    alphaBoolBlend {blendOption=Linear}
    controlRigAssetReference {blueprintRigClass=&/F_MED_Amour_Skirt_CtrlRig_PreSim.F_MED_Amour_Skirt_CtrlRig_PreSim_C}
   }
   n2 = "Transform (Modify) Bone - Bone: dyn_skirt_side_l_base"(ComponentPose=Local To Component(LocalPose=@n1), Alpha=1, Translation="1.409,-0.000,1.135", Rotation="0,0,0", Scale="1,1,1") {
    translationMode=BMM_Additive
    alphaBoolBlend {blendOption=Linear}
    boneToModify {boneName=dyn_skirt_side_l_base}
    translation {x=1.409, z=1.135}
   }
   n3 = "Transform (Modify) Bone - Bone: dyn_skirt_side_r_base"(ComponentPose=@n2, Alpha=1, Translation="-1.958,0.000,1.948", Rotation="5.000,0.000,0.000", Scale="1,1,1") {
    rotationMode=BMM_Additive
    translationMode=BMM_Additive
    alphaBoolBlend {blendOption=Linear}
    boneToModify {boneName=dyn_skirt_side_r_base}
    rotation {pitch=5}
    translation {x=-1.958, z=1.948}
   }
   n4 = "Constraint - dyn_skirt_side_l_base (P)"(ComponentPose=@n3, Alpha=0.3, ConstraintWeights_0=1) {
    alphaBoolBlend {blendOption=Linear}
    boneToModify {boneName=dyn_skirt_side_l_base}
    constraintSetup [{targetBone {boneName=thigh_l}}]
    constraintWeights [1]
   }
   n5 = "Constraint - dyn_skirt_side_r_base (P)"(ComponentPose=@n4, Alpha=0.3, ConstraintWeights_0=1) {
    alphaBoolBlend {blendOption=Linear}
    boneToModify {boneName=dyn_skirt_side_r_base}
    constraintSetup [{targetBone {boneName=thigh_r}}]
    constraintWeights [1]
   }
   n6 = "Constraint - dyn_skirt_fr_l_01 (P)"(ComponentPose=@n5, Alpha=1, ConstraintWeights_0=1) {
    alphaBoolBlend {blendOption=Linear}
    boneToModify {boneName=dyn_skirt_fr_l_01}
    constraintSetup [{targetBone {boneName=thigh_l}}]
    constraintWeights [1]
   }
   n7 = "Constraint - dyn_skirt_fr_r_01 (P)"(ComponentPose=@n6, Alpha=1, ConstraintWeights_0=1) {
    alphaBoolBlend {blendOption=Linear}
    boneToModify {boneName=dyn_skirt_fr_r_01}
    constraintSetup [{targetBone {boneName=thigh_r}}]
    constraintWeights [1]
   }
   n8 = "Constraint - dyn_skirt_side_l_base (P)"(ComponentPose=@n7, Alpha=1, ConstraintWeights_0=1) {
    alphaBoolBlend {blendOption=Linear}
    alphaScaleBiasClamp {bInterpResult=true}
    boneToModify {boneName=dyn_skirt_side_l_base}
    constraintSetup [{targetBone {boneName=thigh_l}}]
    constraintWeights [1]
   }
   n9 = "Constraint - dyn_skirt_bck_l_base (P)"(ComponentPose=@n8, Alpha=1, ConstraintWeights_0=1) {
    alphaBoolBlend {blendOption=Linear}
    alphaScaleBiasClamp {bInterpResult=true}
    boneToModify {boneName=dyn_skirt_bck_l_base}
    constraintSetup [{targetBone {boneName=thigh_l}}]
    constraintWeights [1]
   }
   n10 = "Constraint - dyn_skirt_side_r_base (P)"(ComponentPose=@n9, Alpha=1, ConstraintWeights_0=1) {
    alphaBoolBlend {blendOption=Linear}
    alphaScaleBiasClamp {bInterpResult=true}
    boneToModify {boneName=dyn_skirt_side_r_base}
    constraintSetup [{targetBone {boneName=thigh_r}}]
    constraintWeights [1]
   }
   n11 = "Constraint - dyn_skirt_bck_r_base (P)"(ComponentPose=@n10, Alpha=1, ConstraintWeights_0=1) {
    alphaBoolBlend {blendOption=Linear}
    alphaScaleBiasClamp {bInterpResult=true}
    boneToModify {boneName=dyn_skirt_bck_r_base}
    constraintSetup [{targetBone {boneName=thigh_r}}]
    constraintWeights [1]
   }
   n12 = "Transform (Modify) Bone - Bone: dyn_skirt_side_l_base"(ComponentPose=@n11, Alpha=1, Translation="2,0,3", Rotation="-5,0,0", Scale="1,1,1") {
    rotationMode=BMM_Additive
    translationMode=BMM_Additive
    alphaBoolBlend {blendOption=Linear}
    alphaScaleBiasClamp {bInterpResult=true}
    boneToModify {boneName=dyn_skirt_side_l_base}
    rotation {pitch=-5}
    translation {x=2, z=3}
   }
   n13 = "Transform (Modify) Bone - Bone: dyn_skirt_fr_l_base"(ComponentPose=@n12, Alpha=1, Translation="0,2,2", Rotation="0,0,0", Scale="1,1,1") {
    rotationMode=BMM_Additive
    translationMode=BMM_Additive
    alphaBoolBlend {blendOption=Linear}
    alphaScaleBiasClamp {bInterpResult=true}
    boneToModify {boneName=dyn_skirt_fr_l_base}
    translation {y=2, z=2}
   }
   n14 = "Transform (Modify) Bone - Bone: dyn_skirt_bck_l_base"(ComponentPose=@n13, Alpha=1, Translation="2,-5,3", Rotation="0,0,0", Scale="1,1,1") {
    rotationMode=BMM_Additive
    translationMode=BMM_Additive
    alphaBoolBlend {blendOption=Linear}
    alphaScaleBiasClamp {bInterpResult=true}
    boneToModify {boneName=dyn_skirt_bck_l_base}
    translation=<2 -5 3>
   }
   n15 = "Transform (Modify) Bone - Bone: dyn_skirt_side_r_base"(ComponentPose=@n14, Alpha=1, Translation="-3,0,3", Rotation="0,0,0", Scale="1,1,1") {
    rotationMode=BMM_Additive
    translationMode=BMM_Additive
    alphaBoolBlend {blendOption=Linear}
    alphaScaleBiasClamp {bInterpResult=true}
    boneToModify {boneName=dyn_skirt_side_r_base}
    translation {x=-3, z=3}
   }
   n16 = "Transform (Modify) Bone - Bone: dyn_skirt_bck_r_base"(ComponentPose=@n15, Alpha=1, Translation="-2,-5,3", Rotation="0,0,0", Scale="1,1,1") {
    rotationMode=BMM_Additive
    translationMode=BMM_Additive
    alphaBoolBlend {blendOption=Linear}
    alphaScaleBiasClamp {bInterpResult=true}
    boneToModify {boneName=dyn_skirt_bck_r_base}
    translation=<-2 -5 3>
   }
   n17 = "Transform (Modify) Bone - Bone: dyn_skirt_side_r_base"(ComponentPose=@n16, Alpha=1, Translation="-6.096,-4.889,-0.000", Rotation="0,0,0", Scale="1,1,1") {
    rotationMode=BMM_Additive
    translationMode=BMM_Additive
    alphaBoolBlend {blendOption=Linear}
    alphaScaleBiasClamp {bInterpResult=true}
    boneToModify {boneName=dyn_skirt_side_r_base}
    translation=<-6.096 -4.889>
   }
   n18 = "Transform (Modify) Bone - Bone: dyn_skirt_bck_r_base"(ComponentPose=@n17, Alpha=1, Translation="-6.567,-1.961,0.000", Rotation="0,0,0", Scale="1,1,1") {
    rotationMode=BMM_Additive
    translationMode=BMM_Additive
    alphaBoolBlend {blendOption=Linear}
    alphaScaleBiasClamp {bInterpResult=true}
    boneToModify {boneName=dyn_skirt_bck_r_base}
    translation=<-6.567 -1.961>
   }
   n19 = RigidBody!(ComponentPose=@n18, Alpha, OverrideWorldGravity:Vector=Get Gravity Override(), ExternalForce) -> (Pose) {
    bClampLinearTranslationLimitToRefPose=true
    bOverrideWorldGravity=true
    overridePhysicsAsset=&/F_MED_Amour_Skirt2_Physics
    simulationSpace=BaseBoneSpace
    alphaBoolBlend {blendOption=Linear}
    baseBoneRef {boneName=pelvis}
    componentAppliedLinearAccClamp=<3000 3000 3000>
    componentLinearAccScale=<0.3 0.3 0.2>
    componentLinearVelScale {z=0.6}
    overrideWorldGravity {z=-980}
    simSpaceSettings {
     maxAngularAcceleration=500
     maxAngularVelocity=3
     maxLinearAcceleration=700
     maxLinearVelocity=500
     velocityScaleZ=0
     worldAlpha=0.25
     externalLinearDragV=<1 1>
    }
   }
   n20 = "Constraint - dyn_bow_main (P)"(ComponentPose=@n19, Alpha=1, ConstraintWeights_0=1) {
    alphaBoolBlend {blendOption=Linear}
    boneToModify {boneName=dyn_bow_main}
    constraintSetup [{targetBone {boneName=pelvis}}]
    constraintWeights [1]
   }
   n21 = RigidBody(ComponentPose=@n20, OverrideWorldGravity="0,0,-980") {
    bClampLinearTranslationLimitToRefPose=true
    bOverrideWorldGravity=true
    overridePhysicsAsset=&/F_MED_Amour_Ribbon_Physics
    simulationSpace=BaseBoneSpace
    alphaBoolBlend {blendOption=Linear}
    baseBoneRef {boneName=spine_01}
    componentAppliedLinearAccClamp=<3000 3000 3000>
    componentLinearAccScale=<0.1 0.1 0.1>
    componentLinearVelScale=<0.2 0.2 0.2>
    overrideWorldGravity {z=-980}
    simSpaceSettings {maxAngularAcceleration=1000, maxAngularVelocity=1000, maxLinearAcceleration=1000, maxLinearVelocity=1000}
   }
   n22 = RigidBody(ComponentPose=@n21, OverrideWorldGravity="0,0,-980") {
    bClampLinearTranslationLimitToRefPose=true
    bOverrideWorldGravity=true
    overridePhysicsAsset=&F_MED_Amour_Ribbon_Physics
    simulationSpace=BaseBoneSpace
    alphaBoolBlend {blendOption=Linear}
    baseBoneRef {boneName=spine_01}
    componentAppliedLinearAccClamp=<3000 3000 3000>
    componentLinearAccScale=<0.7 0.7 0.5>
    componentLinearVelScale=<0.2 0.2 0.2>
    overrideWorldGravity {z=-980}
    simSpaceSettings {maxAngularAcceleration=1000, maxAngularVelocity=1000, maxLinearAcceleration=1000, maxLinearVelocity=1000}
   }
   n23 = "Transform (Modify) Bone - Bone: dyn_shoulderRings_l"(ComponentPose=@n22, Alpha=1, Translation="-0.818,-3.510,-0.685", Rotation="16.559,-26.688,-29.442", Scale="1,1.1,1") {
    rotationMode=BMM_Additive
    scaleMode=BMM_Replace
    translationMode=BMM_Additive
    translationSpace=BCS_BoneSpace
    alphaBoolBlend {blendOption=Linear}
    boneToModify {boneName=dyn_shoulderRings_l}
    rotation=<rot 16.559 -26.688 -29.442>
    scale {y=1.1}
    translation=<-0.818 -3.51 -0.685>
   }
   n24 = "Transform (Modify) Bone - Bone: dyn_shoulderRings_r"(ComponentPose=@n23, Alpha=1, Translation="-0.869,4.232,-0.014", Rotation="9.154,21.847,-35.315", Scale="1,1.1,1") {
    rotationMode=BMM_Additive
    scaleMode=BMM_Replace
    translationMode=BMM_Additive
    translationSpace=BCS_BoneSpace
    alphaBoolBlend {blendOption=Linear}
    boneToModify {boneName=dyn_shoulderRings_r}
    rotation=<rot 9.154 21.847 -35.315>
    scale {y=1.1}
    translation=<-0.869 4.232 -0.014>
   }
   n25 = "Transform (Modify) Bone - Bone: dyn_shoulderRings_l"(ComponentPose=@n24, Alpha=1, Translation="0,0,0", Rotation="0,0,0", Scale="1,1,1") {
    translationMode=BMM_Additive
    translationSpace=BCS_BoneSpace
    alphaBoolBlend {blendOption=Linear}
    boneToModify {boneName=dyn_shoulderRings_l}
   }
   n26 = "Transform (Modify) Bone - Bone: dyn_shoulderRings_r"(ComponentPose=@n25, Alpha=1, Translation="0,0,0", Rotation="0,0,0", Scale="1,1,1") {
    translationMode=BMM_Additive
    translationSpace=BCS_BoneSpace
    alphaBoolBlend {blendOption=Linear}
    boneToModify {boneName=dyn_shoulderRings_r}
   }
   n27 = "Transform (Modify) Bone - Bone: dyn_shoulderRings_l"(ComponentPose=@n26, Alpha=1, Translation="0,0,-2", Rotation="0,0,0", Scale="1,1,1") {
    translationMode=BMM_Additive
    translationSpace=BCS_BoneSpace
    alphaBoolBlend {blendOption=Linear}
    boneToModify {boneName=dyn_shoulderRings_l}
    translation {z=-2}
   }
   n28 = "Transform (Modify) Bone - Bone: dyn_shoulderRings_r"(ComponentPose=@n27, Alpha=1, Translation="0,0,-2", Rotation="0,0,0", Scale="1,1,1") {
    translationMode=BMM_Additive
    translationSpace=BCS_BoneSpace
    alphaBoolBlend {blendOption=Linear}
    boneToModify {boneName=dyn_shoulderRings_r}
    translation {z=-2}
   }
   n29 = "Transform (Modify) Bone - Bone: dyn_skirt_fr_l_base"(ComponentPose=@n28, Alpha=1, Translation="0.000,1.082,1.461", Rotation="-0.000,0.000,15.000", Scale="1,1,1") {
    rotationMode=BMM_Additive
    translationMode=BMM_Additive
    alphaBoolBlend {blendOption=Linear}
    alphaScaleBiasClamp {bInterpResult=true}
    boneToModify {boneName=dyn_skirt_fr_l_base}
    rotation {roll=15}
    translation {y=1.082, z=1.461}
   }
   n30 = "Transform (Modify) Bone - Bone: dyn_skirt_side_l_base"(ComponentPose=@n29, Alpha=1, Translation="-0.048,-0.000,1.253", Rotation="0,0,0", Scale="1,1,1") {
    rotationMode=BMM_Additive
    translationMode=BMM_Additive
    alphaBoolBlend {blendOption=Linear}
    alphaScaleBiasClamp {bInterpResult=true}
    boneToModify {boneName=dyn_skirt_side_l_base}
    translation {x=-0.048, z=1.253}
   }
   n31 = "Transform (Modify) Bone - Bone: dyn_skirt_bck_l_base"(ComponentPose=@n30, Alpha=1, Translation="1,-2,0", Rotation="0,0,0", Scale="1,1,1") {
    rotationMode=BMM_Additive
    translationMode=BMM_Additive
    alphaBoolBlend {blendOption=Linear}
    alphaScaleBiasClamp {bInterpResult=true}
    boneToModify {boneName=dyn_skirt_bck_l_base}
    translation=<1 -2>
   }
   n32 = "Transform (Modify) Bone - Bone: dyn_skirt_bck_mid_base"(ComponentPose=@n31, Alpha=1, Translation="0,-2,0", Rotation="0,0,0", Scale="1,1,1") {
    rotationMode=BMM_Additive
    translationMode=BMM_Additive
    alphaBoolBlend {blendOption=Linear}
    alphaScaleBiasClamp {bInterpResult=true}
    boneToModify {boneName=dyn_skirt_bck_mid_base}
    translation {y=-2}
   }
   n33 = "Transform (Modify) Bone - Bone: dyn_skirt_bck_r_base"(ComponentPose=@n32, Alpha=1, Translation="-1,-2,0", Rotation="0,0,0", Scale="1,1,1") {
    rotationMode=BMM_Additive
    translationMode=BMM_Additive
    alphaBoolBlend {blendOption=Linear}
    alphaScaleBiasClamp {bInterpResult=true}
    boneToModify {boneName=dyn_skirt_bck_r_base}
    translation=<-1 -2>
   }
   n34 = Control Rig(Source=Component To Local(ComponentPose=@n33)) {
    alphaBoolBlend {blendOption=Linear}
    controlRigAssetReference {blueprintRigClass=&/F_MED_Amour_ArmRIng_CtrlRig_PostSim.F_MED_Amour_ArmRIng_CtrlRig_PostSim_C}
   }
   Output Pose(Result=@n34) {layerGroup=None}
  ]}
 EventGraph {nodes [
   Event BlueprintUpdateAnimation()
   > n1 = Gravity Override State Function DEPRECATED(CharacterPartAnimInstance=Self-Reference(), GameTimeInSeconds=GetGameTimeInSeconds(), GravityOverrideParams=Get Gravity Override Params())
   > Set Gravity Override(Gravity Override=@n1)
   > n2 = Branch(Condition=Get bIsPlayingEmote())
   > n3 = Set frontend ring(frontend ring:double=0) -> (Output_Get:double)
   n4 = GetGameTimeInSeconds()
   n5 = "Sin (Radians)"(A=float * float(A=float * float(A=0.6, B=3.3), B=@n4))
   Set Floater 1(execute=[@n21.then, @n22.then, @n3.then], Floater 1:Vector=MakeVector(X=float * float(A=@n5, B=0.7), Y=0)) -> (Output_Get:Vector)
   n6 = "Sin (Radians)"(A=float * float(A=float * float(A=0.6, B=3.5), B=@n4))
   > Set Floater 2(Floater 2:Vector=MakeVector(X=float * float(A=@n6, B=0.8), Y=0)) -> (Output_Get:Vector)
   > n7 = Branch(Condition=IsCosmeticObjectDisplayedInFrontendContext!(CosmeticObject:Object Reference) -> (ReturnValue:bool))
   > n8 = Set Ribbon Normal Alpha(Ribbon Normal Alpha:double=0.5) -> (Output_Get:double)
   n9 = Branch(execute=[@n16.then, @n8.then], Condition=CheckForBackbling())
   > Set Ribbon Backpack Alpha(Ribbon Backpack Alpha:double=1) -> (Output_Get:double)
   > n10 = Set Ribbon Normal Alpha(Ribbon Normal Alpha=0)
   n11 = Branch(execute=[@n15.then, @n10.then], Condition=Get bIsPlayingEmote())
   > Set Funky Leg Emote Alpha(Funky Leg Emote Alpha:double=1) -> (Output_Get:double)
   > n12 = Set rban emote alpha(rban emote alpha:double=0.5) -> (Output_Get:double)
   n13 = Branch(execute=[@n14.then, @n12.then], Condition=Get bIsSliding())
   > Set sliding alpha(sliding alpha:double=1) -> (Output_Get:double)
   Set sliding alpha(execute=@n13.else, sliding alpha=0)
   Set Funky Leg Emote Alpha(execute=@n11.else, Funky Leg Emote Alpha=0)
   > n14 = Set rban emote alpha(rban emote alpha=1)
   Set Ribbon Backpack Alpha(execute=@n9.else, Ribbon Backpack Alpha=0)
   > n15 = Set Ribbon Normal Alpha(Ribbon Normal Alpha=1)
   n16 = Set Ribbon Normal Alpha(execute=@n7.else, Ribbon Normal Alpha=1)
   n17 = GetDeltaTransformFromRefPose(self=GetOwnerSkeletalMeshComponent(), BoneName=upperarm_r, BaseName=clavicle_r)
   n18 = MapRangeClamped(Value=BreakRotator(InRot=BreakTransform(InTransform=@n17).Rotation).Pitch, InRangeA=20, InRangeB=25)
   n19 = "To Float (Boolean)"(InBool=IsCosmeticObjectDisplayedInFrontendContext())
   n20 = Branch(execute=@n2.else, Condition="float > float"(A=float * float(A=@n18, B=@n19) {is=title K2Node_PromotableOperator}))
   > n21 = Set frontend ring(frontend ring=1)
   n22 = Set frontend ring(execute=@n20.else, frontend ring=0)
   n23 = Branch(Condition=IsCosmeticObjectDisplayedInFrontendContext())
   > Set frontend ring(frontend ring=1)
   Set frontend ring(execute=@n23.else, frontend ring=0)
  ]}
]
variables [
 Floater 1 {type=Vector}
 Floater 2 {type=Vector}
 frontend ring {type=number}
 Funky Leg Emote Alpha {type=number}
 Gravity Override {type=Vector}
 Gravity Override Params {type=GravityOverrideParamsStruct}
 NewVar {type=number}
 normal ring {type=number}
 rban emote alpha {type=number}
 Ribbon Backpack Alpha {type=number}
 Ribbon Normal Alpha {type=number}
 sliding alpha {type=number}
]
