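# Physics asset for the earring part of the F_MED_Amour female player cosmetic, defining physical bodies for two dangling earrings.
# Each dyn_earring bone has a small capsule constrained to the head with a swing cone and zero twist limit, slightly wider on the right.
# Larger capsules on head, neck_02 and both clavicles approximate the surrounding body.
# Head-to-neck rotation is free, while the clavicle bodies have no constraints.
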PhysicsAsset /BRCosmetics/Characters/Player/Female/Medium/Bodies/F_MED_Amour/Meshes/Parts/F_MED_Amour_Earring_Physics
bodies {
 clavicle_l [{shapeType=Capsule, length=6, radius=6.25, shapeName=clavicle_l_capsule, center=<6.3009 0 0>, rotation=<rot 90 0 0>}]
 clavicle_r [{shapeType=Capsule, length=6, radius=6.25, shapeName=clavicle_r_capsule, center=<-5.6376 0 0>, rotation=<rot 90 0 0>}]
 dyn_earring_l [{shapeType=Capsule, length=2.25, radius=1, shapeName=dyn_earring_l_capsule, center=<2.2737 -0.064 -0.1206>, rotation=<rot 90 0 0>}]
 dyn_earring_r [{shapeType=Capsule, length=2.25, radius=1, shapeName=dyn_earring_r_capsule, center=<2.2737 0.1703 -0.1206>, rotation=<rot 90 0 0>}]
 head [{shapeType=Capsule, length=4, radius=6.25, shapeName=head_capsule, center=<3.7637 0.8404 0>, rotation=<rot 90 0 0>}]
 neck_02 [{shapeType=Capsule, length=6, radius=4.75, shapeName=neck_02_capsule, center=<0 0.3787 0>, rotation=<rot 90 0 0>}]
}
constraints [
 {bone1Name=dyn_earring_l, bone2Name=head, swing1LimitDegrees=45, swing2LimitDegrees=45, twistLimitDegrees=0}
 {bone1Name=dyn_earring_r, bone2Name=head, swing1LimitDegrees=50, swing2LimitDegrees=50, twistLimitDegrees=0}
 {
  bone1Name=head
  bone2Name=neck_02
  swing1LimitDegrees=45
  swing1Motion=Free
  swing2LimitDegrees=45
  swing2Motion=Free
  twistLimitDegrees=45
  twistMotion=Free
 }
]
