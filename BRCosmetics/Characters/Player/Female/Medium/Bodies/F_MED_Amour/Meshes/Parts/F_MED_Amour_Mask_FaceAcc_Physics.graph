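# Physics asset for the mask face accessory of the female medium Amour player cosmetic, covering its left and right dynamic mask tails.
# Each tail is a three-body capsule chain (base, 01, 02) constrained from the head, with tight swing limits and zero-degree twist limits.
# Capsules on the head, upper spine, clavicles and upper arms form a coarse upper-body shape around the tails, with free rotation at the head and shoulder joints.
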PhysicsAsset /BRCosmetics/Characters/Player/Female/Medium/Bodies/F_MED_Amour/Meshes/Parts/F_MED_Amour_Mask_FaceAcc_Physics
bodies {
 clavicle_l [{shapeType=Capsule, length=6, radius=8, shapeName=clavicle_l_capsule, center=<7.8382 0 -2.3965>, rotation=<rot -85 179.9995 -179.9995>}]
 clavicle_r [{shapeType=Capsule, length=6, radius=8, shapeName=clavicle_r_capsule, center=<-7.5761 0 2.7017>, rotation=<rot 75 -0.0003 -0.0004>}]
 dyn_maskTail_base_l [{shapeType=Capsule, length=0, radius=3.25, center=<0 0 0.2302>}]
 dyn_maskTail_base_r [{shapeType=Capsule, length=0, radius=3.25, center=<0 0 0.2302>}]
 dyn_maskTail_l_01 [{shapeType=Capsule, length=2, radius=2, shapeName=dyn_maskTail_l_01_capsule, center=<0.5439 0 0.198>, rotation=<rot -70 0.0002 -0.0001>}]
 dyn_maskTail_l_02 [{shapeType=Capsule, length=1.25, radius=2, shapeName=dyn_maskTail_l_02_capsule, center=<0.7002 0 -0.0496>, rotation=<rot 85 -0.0001 -0.0006>}]
 dyn_maskTail_r_01 [{shapeType=Capsule, length=2, radius=2, shapeName=dyn_maskTail_r_01_capsule, center=<0.5439 0 0.198>, rotation=<rot -70 -0.0003 0.0002>}]
 dyn_maskTail_r_02 [{shapeType=Capsule, length=1.25, radius=2, shapeName=dyn_maskTail_r_02_capsule, center=<0.7002 0 -0.0496>, rotation=<rot 85 0.0001 0.0006>}]
 head [{shapeType=Capsule, length=0, radius=5.5, center=<3.6781 0 0>}]
 spine_05 [{shapeType=Capsule, length=2.75, radius=11, shapeName=spine_05_capsule, center=<4.0337 -0.7682 0>}]
 upperarm_l [{shapeType=Capsule, length=0, radius=8.255, shapeName=upperarm_l_capsule, center=<4.942 0 1.9348>, rotation=<rot -0.0003 -89.9997 75>}]
 upperarm_r [{shapeType=Capsule, length=0, radius=8.255, shapeName=upperarm_r_capsule, center=<-4.942 0 -1.9348>, rotation=<rot -0.0004 -89.9996 -105>}]
}
constraints [
 {bone1Name=dyn_maskTail_base_r, bone2Name=head, swing1LimitDegrees=10, swing2LimitDegrees=15, twistLimitDegrees=0}
 {bone1Name=dyn_maskTail_base_l, bone2Name=head, swing1LimitDegrees=10, swing2LimitDegrees=15, twistLimitDegrees=0}
 {bone1Name=dyn_maskTail_l_01, bone2Name=dyn_maskTail_base_l, swing1LimitDegrees=10, swing2LimitDegrees=15, twistLimitDegrees=0}
 {bone1Name=dyn_maskTail_l_02, bone2Name=dyn_maskTail_l_01, swing1LimitDegrees=10, swing2LimitDegrees=15, twistLimitDegrees=0}
 {bone1Name=dyn_maskTail_r_01, bone2Name=dyn_maskTail_base_r, swing1LimitDegrees=10, swing2LimitDegrees=15, twistLimitDegrees=0}
 {bone1Name=dyn_maskTail_r_02, bone2Name=dyn_maskTail_r_01, swing1LimitDegrees=10, swing2LimitDegrees=15, twistLimitDegrees=0}
 {
  bone1Name=head
  bone2Name=spine_05
  swing1LimitDegrees=45
  swing1Motion=Free
  swing2LimitDegrees=45
  swing2Motion=Free
  twistLimitDegrees=45
  twistMotion=Free
 }
 {bone1Name=clavicle_l, bone2Name=spine_05, swing1LimitDegrees=45, swing2LimitDegrees=45, twistLimitDegrees=45}
 {bone1Name=clavicle_r, bone2Name=spine_05, swing1LimitDegrees=45, swing2LimitDegrees=45, twistLimitDegrees=45}
 {
  bone1Name=upperarm_r
  bone2Name=clavicle_r
  swing1LimitDegrees=45
  swing1Motion=Free
  swing2LimitDegrees=45
  swing2Motion=Free
  twistLimitDegrees=45
  twistMotion=Free
 }
 {
  bone1Name=upperarm_l
  bone2Name=clavicle_l
  swing1LimitDegrees=45
  swing1Motion=Free
  swing2LimitDegrees=45
  swing2Motion=Free
  twistLimitDegrees=45
  twistMotion=Free
 }
]
