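# Post-process style anim blueprint for the face accessory of the female F_MED_Amour mask, copying the pose from the attached parent mesh.
# The output chain only applies additive bone-space offsets to the left and right mask ring bones.
# Anim Dynamics rings, head-constrained tail bases and a RigidBody simulation exist but are not wired to the output.
# Each update sets a gravity override via a deprecated function and two sine-wave floater vectors, with no visible reader.
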AnimBlueprint /BRCosmetics/Characters/Player/Female/Medium/Bodies/F_MED_Amour/Meshes/Parts/F_MED_Amour_Mask_FaceAcc_AnimBP
target_skeleton=&/F_MED_Amour_Mask_FaceAcc_Skeleton
graphs [
 AnimGraph {nodes [
   n1 = Anim Dynamics - Bone: dyn_maskRings_l(Alpha=1, ExternalForce="0,0,0") {
    angularSpringConstant=300
    bLinearSpring=true
    bUseGravityOverride=true
    linearSpringConstant=50
    simulationSpace=BoneRelative
    alphaBoolBlend {blendOption=Linear}
    boundBone {boneName=dyn_maskRings_l}
    componentLinearAccScale {x=0.1, z=0.1}
    gravityOverride {z=-100}
    physicsBodyDefinitions [{
      boundBone {boneName=dyn_maskRings_l}
      boxExtents=<100 100 100>
      constraintSetup {angularLimitsMax=<0 0 0>, angularLimitsMin=<0 0 0>, angularTarget=<0 0 0>, linearAxesMax=<1 0 0>, linearAxesMin=<0 0 -1>}
      localJointOffset=<0.1 0 0>
     }]
    relativeSpaceBone {boneName=head}
   }
   Anim Dynamics - Bone: dyn_maskRings_r(ComponentPose=@n1, Alpha=1, ExternalForce="0,0,0") {
    angularSpringConstant=300
    bLinearSpring=true
    bUseGravityOverride=true
    linearSpringConstant=50
    simulationSpace=BoneRelative
    alphaBoolBlend {blendOption=Linear}
    boundBone {boneName=dyn_maskRings_r}
    componentLinearAccScale {x=0.1, z=0.1}
    gravityOverride {z=-100}
    physicsBodyDefinitions [{
      boundBone {boneName=dyn_maskRings_r}
      boxExtents=<100 100 100>
      constraintSetup {angularLimitsMax=<0 0 0>, angularLimitsMin=<0 0 0>, angularTarget=<0 0 0>, linearAxesMax=<1 0 0>, linearAxesMin=<0 0 -1>}
      localJointOffset=<0.1 0 0>
     }]
    relativeSpaceBone {boneName=head}
   }
   n2 = "Constraint - dyn_maskTail_base_r (P)"(Alpha=1, ConstraintWeights_0=1) {
    alphaBoolBlend {blendOption=Linear}
    boneToModify {boneName=dyn_maskTail_base_r}
    constraintSetup [{targetBone {boneName=head}}]
    constraintWeights [1]
   }
   n3 = "Constraint - dyn_maskTail_base_l (P)"(ComponentPose=@n2, Alpha=1, ConstraintWeights_0=1) {
    alphaBoolBlend {blendOption=Linear}
    boneToModify {boneName=dyn_maskTail_base_l}
    constraintSetup [{targetBone {boneName=head}}]
    constraintWeights [1]
   }
   RigidBody!(ComponentPose=@n3, Alpha, OverrideWorldGravity:Vector="0,0,-980", ExternalForce) -> (Pose) {
    bClampLinearTranslationLimitToRefPose=true
    bOverrideWorldGravity=true
    overridePhysicsAsset=&/F_MED_Amour_Mask_FaceAcc_Physics
    simulationSpace=BaseBoneSpace
    alphaBoolBlend {blendOption=Linear}
    baseBoneRef {boneName=neck_02}
    componentLinearVelScale=<0.4 0.4 0.4>
    overrideWorldGravity {z=-980}
    simSpaceSettings {maxAngularAcceleration=1000, maxAngularVelocity=1000, maxLinearAcceleration=1000, maxLinearVelocity=1000}
   }
   n4 = "Transform (Modify) Bone - Bone: dyn_maskRings_l"(
    ComponentPose=Local To Component(LocalPose=Copy Pose From Mesh() {bUseAttachedParent=true})
    Alpha=1
    Translation="1.212,0.018,-1.147"
    Rotation="0,0,0"
    Scale="1,1,1"
   ) {
    translationMode=BMM_Additive
    translationSpace=BCS_BoneSpace
    alphaBoolBlend {blendOption=Linear}
    boneToModify {boneName=dyn_maskRings_l}
    translation=<1.212 0.018 -1.147>
   }
   n5 = "Transform (Modify) Bone - Bone: dyn_maskRings_r"(ComponentPose=@n4, Alpha=1, Translation="1.000,0.000,-1.588", Rotation="0,0,0", Scale="1,1,1") {
    translationMode=BMM_Additive
    translationSpace=BCS_BoneSpace
    alphaBoolBlend {blendOption=Linear}
    boneToModify {boneName=dyn_maskRings_r}
    translation {x=1, z=-1.588}
   }
   n6 = "Transform (Modify) Bone - Bone: dyn_maskRings_l"(ComponentPose=@n5, Alpha=0.9, Translation="0,0,0", Rotation="0,0,0", Scale="1,1,1") {translationMode=BMM_Additive, translationSpace=BCS_BoneSpace, alphaBoolBlend {blendOption=Linear}, boneToModify {boneName=dyn_maskRings_l}}
   n7 = "Transform (Modify) Bone - Bone: dyn_maskRings_r"(ComponentPose=@n6, Alpha=0.9, Translation="0,0,0", Rotation="0,0,0", Scale="1,1,1") {translationMode=BMM_Additive, translationSpace=BCS_BoneSpace, alphaBoolBlend {blendOption=Linear}, boneToModify {boneName=dyn_maskRings_r}}
   Output Pose(Result=Component To Local(ComponentPose=@n7)) {layerGroup=None}
  ]}
 EventGraph {nodes [
   Event BlueprintUpdateAnimation()
   > n1 = Gravity Override State Function DEPRECATED(CharacterPartAnimInstance=Self-Reference(), GameTimeInSeconds=GetGameTimeInSeconds(), GravityOverrideParams=Get Gravity Override Params())
   > Set Gravity Override(Gravity Override=@n1)
   n2 = GetGameTimeInSeconds()
   n3 = "Sin (Radians)"(A=float * float(A=float * float(A=0.5, B=3.3), B=@n2))
   > Set Floater 1(Floater 1:Vector=MakeVector(X=float * float(A=@n3, B=0.8), Y=0)) -> (Output_Get:Vector)
   n4 = "Sin (Radians)"(A=float * float(A=float * float(A=0.6, B=3.5), B=@n2))
   > Set Floater 2(Floater 2:Vector=MakeVector(X=float * float(A=@n4, B=0.8), Y=0)) -> (Output_Get:Vector)
  ]}
]
variables [
 Floater 1 {type=Vector}
 Floater 2 {type=Vector}
 Gravity Override {type=Vector}
 Gravity Override Params {type=GravityOverrideParamsStruct}
]
